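# Sound cue for ball impact sounds in the SunnyLoma ball gameplay.
# A Random node picks one of six RL_Ball_Impact waves with equal weight and feeds a Mixer at the root, whose second input is empty.
# Playback uses the SunnyLoma ball attenuation asset with a set max distance and is routed to the Fort_Gadgets_Items sound class.
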SoundCue /SunnyLomaBallGameplay/Assets/Audio/SunnyLoma_Ball_Impact_Cue
root=Mixer_0
nodes [
 {id=Mixer_0, children [Random_0,null], properties {inputVolume [1,1]}}
 {id=Random_0, children [WavePlayer_0,WavePlayer_1,WavePlayer_2,WavePlayer_3,WavePlayer_4,WavePlayer_5], properties {weights [1,1,1,1,1,1]}}
 {id=WavePlayer_0, properties {soundWaveAssetPtr=&/RL_Ball_Impact_06}}
 {id=WavePlayer_1, properties {soundWaveAssetPtr=&/RL_Ball_Impact_05}}
 {id=WavePlayer_2, properties {soundWaveAssetPtr=&/RL_Ball_Impact_04}}
 {id=WavePlayer_3, properties {soundWaveAssetPtr=&/RL_Ball_Impact_03}}
 {id=WavePlayer_4, properties {soundWaveAssetPtr=&/RL_Ball_Impact_02}}
 {id=WavePlayer_5, properties {soundWaveAssetPtr=&/RL_Ball_Impact_01}}
]
properties {
 attenuationSettings=&/SunnyLoma_Ball_Att
 maxDistance=4600
 soundClassObject=&Fort_Gadgets_Items
}
