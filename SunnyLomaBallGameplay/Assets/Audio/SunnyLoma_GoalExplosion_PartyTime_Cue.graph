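# Sound cue for the Party Time goal explosion in SunnyLomaBall gameplay.
# A single wave player feeds the FN_RL_GoalExplosion_PartyTime_01 wave into a mixer whose second input is empty.
# The cue plays at reduced volume in the Fort_Gadgets_Items sound class, spatialized through the goal explosion attenuation asset.
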SoundCue /SunnyLomaBallGameplay/Assets/Audio/SunnyLoma_GoalExplosion_PartyTime_Cue
root=Mixer_0
nodes [
 {id=Mixer_0, children [WavePlayer_0,null], properties {inputVolume [1,1]}}
 {id=WavePlayer_0, properties {soundWaveAssetPtr=&/FN_RL_GoalExplosion_PartyTime_01}}
]
properties {
 attenuationSettings=&/SunnyLoma_GoalExplosion_Att
 maxDistance=7500
 soundClassObject=&Fort_Gadgets_Items
 volumeMultiplier=0.7
}
