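# Material function that drives a ball vanish effect, outputting opacity and emissive from the Ball_Visibility parameter.
# A vertex-interpolated, normalized local-Z gradient is perturbed by the green channel of a tiled texture to give an uneven dissolve front.
# Opacity smoothsteps that gradient shifted by visibility.
# Emissive uses a SphereMask as a scalar band, comparing the remapped gradient to visibility to tint a glow, orange by default.
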MaterialFunction /SunnyLomaBallGameplay/Assets/Materials/MF_Ball_Vanish
graph {nodes [
  n1 = ScalarParameter() {defaultValue=1, parameterName=Ball_Visibility}
  n2 = Multiply(A=@n1) {constB=2}
  n3 = Subtract(A=LocalPosition() {localOrigin=InstancePreSkinning}.Z, B=ScalarParameter() {defaultValue=-176.4934, parameterName=offfset})
  n4 = ScalarParameter() {defaultValue=491.0836, parameterName=adfga}
  n5 = TextureCoordinate() {uTiling=2, vTiling=2}
  n6 = Add(
   A=VertexInterpolator(VS=Divide(A=@n3, B=@n4)).PS
   B=Divide(A=TextureSample(UVs=@n5) {samplerType=SAMPLERTYPE_LinearColor, texture=/SunnyLomaBallGameplay/Assets/Textures/T_Ball_Tilling02_M}.G) {constB=6}
  )
  n7 = Subtract(A=@n6, B=Constant() {r=1})
  FunctionOutput(None=SmoothStep(Value=Add(A=@n2, B=@n7))) {outputName=Opacity}
  n8 = LinearInterpolate(A=ScalarParameter() {defaultValue=0.7898, parameterName=offset}, Alpha=@n6) {constA=1.2, constB=0.2}
  n9 = ScalarParameter() {defaultValue=0.0997, parameterName=FadeRad}
  FunctionOutput(None=Multiply(
    A=VectorParameter() {parameterName=Emissive, defaultValue=<rgba 5 1.0005 0 1>}.RGB
    B=SphereMask(A=@n8, B=@n1, Radius=@n9) {attenuationRadius=0.2, hardnessPercent=0}
   )) {outputName=emissive}
 ]}
