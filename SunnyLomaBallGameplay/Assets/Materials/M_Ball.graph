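# Nanite-enabled master material for gameplay balls, with parameterized base color, packed PBR mask, normal and tiling.
# A grunge texture combines with the mask's red channel to tint base color toward DirtColor, and also modulates roughness.
# Optional emissive uses a sharpened time-sine pulse between two intensities, masked by alpha, plus MF_Ball_Vanish's emissive.
# A static switch adds XY-only scaled detail normals.
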Material /SunnyLomaBallGameplay/Assets/Materials/M_Ball
graph {
 properties {bAutomaticallySetUsageInEditor=false, bUsedWithNanite=true, ditherOpacityMask=true, usedShadingModels=Default Lit}
 nodes [
  n1 = VectorParameter() {group=Ball, parameterName=EmissiveColor, defaultValue=<rgba 1 0.0638 0 1>}
  n2 = Sine(None=Multiply(A=Time() {bOverride_Period=true, period=2}, B=Constant() {r=1}))
  n3 = LinearInterpolate(
   A=Multiply(A=@n1.RGB, B=ScalarParameter() {group=Ball, parameterName=EmissiveAmt})
   B=@n1.RGB
   Alpha=Saturate(None=ConstantBiasScale(None=@n2) {bias=-0.5, scale=6})
  )
  n4 = Multiply(A=TextureCoordinate(), B=ScalarParameter() {defaultValue=1, group=Ball, parameterName=TextureTiling})
  n5 = TextureSampleParameter2D(UVs=@n4) {
   group=Ball
   parameterName=MaskPBR
   samplerType=SAMPLERTYPE_LinearColor
   texture=/Game/Global/Textures/Default/Blanks/T_BlackLinearColor
   channelNames {a=EmissiveMask, b=Metallic, g=Roughness, r=Spec}
  }
  Add(
   A=StaticSwitchParameter(True=Multiply(A=@n3, B=@n5.EmissiveMask), False=Constant()) {group=Ball, parameterName=useEmissive}
   B=MaterialFunctionCall() {materialFunction=&/MF_Ball_Vanish}.emissive
  )
  Constant() {r=2}
  n6 = TextureSample(UVs=TextureCoordinate() {uTiling=2, vTiling=2}) {texture=/SunnyLomaBallGameplay/Assets/Textures/T_Ball_Grunge_M}
  n7 = Saturate(None=Multiply(A=Power(Base=@n6.G) {constExponent=4}) {constB=8})
  Multiply(A=Saturate(None=ConstantBiasScale(None=@n7) {bias=0.155, scale=4}), B=@n5.Roughness)
  n8 = VectorParameter() {parameterName=DirtColor, defaultValue=<rgba 0.1146 0.0589 0.0489 1>}
  n9 = Add(A=Multiply(A=@n5.Spec) {constB=2}) {constB=0.025}
  n10 = Saturate(None=Subtract(A=Saturate(None=@n9), B=@n7))
  Reroute(None=Multiply(
    A=LinearInterpolate(A=@n8.RGB, Alpha=@n10)
    B=TextureSampleParameter2D(UVs=@n4) {group=Ball, parameterName=BaseColor, texture=&T_BlacksRGB}.RGB
   ))
  n11 = TextureSampleParameter2D(UVs=@n4) {group=Ball, parameterName=Normal, samplerType=SAMPLERTYPE_Normal, texture=&FlatNormalBlue}
  n12 = TextureSampleParameter2D(UVs=Multiply(A=TextureCoordinate(), B=ScalarParameter() {defaultValue=1, group=Ball, parameterName=DetailTiling})) {group=Ball, parameterName=DetailNormal, samplerType=SAMPLERTYPE_Normal, texture=&FlatNormalBlue}
  n13 = Multiply(A=Constant3Vector() {constant=<rgba 0.1 0.1 0 0>}, B=ScalarParameter() {defaultValue=1, parameterName=DetailAmt})
  StaticSwitchParameter(True=Add(A=@n11.RGB, B=Multiply(A=@n12.RGB, B=@n13)), False=@n11.RGB) {group=Ball, parameterName=useDetailNormal}
 ]
}
