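# Opaque-lit ball material with a tiled mask, normal and tint, flagged for Nanite use.
# A UV1 AO texture combines with the tiled mask channels into a wear factor that blends roughness and dirt color.
# The UV2_AO switch applies that dirt to base color and masks specular, while another switch adds a separately tiled detail normal.
# MF_Ball_Vanish is included alongside dithered opacity masking.
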Material /SunnyLomaBallGameplay/Assets/Materials/M_Ball_SimpleTile
graph {
 properties {bAutomaticallySetUsageInEditor=false, bUsedWithNanite=true, ditherOpacityMask=true, usedShadingModels=Default Lit}
 nodes [
  n1 = Multiply(A=TextureCoordinate(), B=ScalarParameter() {defaultValue=1, group=Ball, parameterName=TextureTiling})
  n2 = TextureSampleParameter2D(UVs=@n1) {parameterName=Mask, samplerType=SAMPLERTYPE_LinearColor, texture=/SunnyLomaBallGameplay/Assets/Textures/T_Ball_Tiling01_M}
  n3 = Multiply(
   A=TextureSample() {constCoordinate=1, texture=/SunnyLomaBallGameplay/Assets/Textures/T_Ball_TilingAO}.R
   B=ScalarParameter() {defaultValue=1.6, parameterName=AOMaskMultiply}
  )
  n4 = Multiply(A=Saturate(None=Add(A=@n2.G) {constB=0.5}), B=Saturate(None=Subtract(A=@n3, B=@n2.R)))
  LinearInterpolate(B=ScalarParameter() {group=Ball, parameterName=Roughness}, Alpha=@n4) {constA=0.55}
  MaterialFunctionCall() {materialFunction=&/MF_Ball_Vanish}
  ScalarParameter() {group=Ball, parameterName=Metallic}
  n5 = TextureSampleParameter2D(UVs=@n1) {group=Ball, parameterName=Normal, samplerType=SAMPLERTYPE_Normal, texture=&FlatNormalBlue}
  n6 = TextureSampleParameter2D(UVs=Multiply(A=TextureCoordinate(), B=ScalarParameter() {defaultValue=1, group=Ball, parameterName=DetailTiling})) {group=Ball, parameterName=DetailNormal, samplerType=SAMPLERTYPE_Normal, texture=&FlatNormalBlue}
  n7 = Multiply(A=Constant3Vector() {constant=<rgba 0.1 0.1 0 0>}, B=ScalarParameter() {defaultValue=1, parameterName=DetailAmt})
  StaticSwitchParameter(True=Add(A=@n5.RGB, B=Multiply(A=@n6.RGB, B=@n7)), False=@n5.RGB) {group=Ball, parameterName=useDetailNormal}
  n8 = VectorParameter() {parameterName=DirtColor, defaultValue=<rgba 0.4 0.3013 0.136 1>}
  n9 = Saturate(None=@n4)
  n10 = VectorParameter() {group=Ball, parameterName=BC, defaultValue=<rgba 0 0 0 1>}
  StaticSwitchParameter(True=Multiply(A=LinearInterpolate(A=@n8.RGB, Alpha=@n9), B=@n10.RGB), False=@n10.RGB) {parameterName=UV2_AO}
  n11 = ConstantBiasScale(None=@n2.B) {bias=0.05, scale=8}
  n12 = ScalarParameter() {group=Ball, parameterName=Specular}
  StaticSwitchParameter(True=Multiply(A=Saturate(None=@n11), B=Multiply(A=@n4, B=@n12)), False=@n12) {parameterName=UV2_AO}
 ]
}
